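# Unlit, two-sided translucent material for skydiving drift contrails on mesh particle ribbons.
# Two panning detail textures, randomized per particle and reshaped by switchable UV direction, drive a curve-atlas color and emissive.
# Opacity combines detail A with a pinched laser-trail texture, soft mask, particle alpha and near-camera fade.
# World position offset adds noise wobble and gravity weighted along the trail; low quality swaps in one panning texture and skips camera fade.
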Material /BRCosmetics/Athena/Effects/Materials/Skydiving/P_Contrails_Drift/M_Contrail_MeshRibbon
graph {
 properties {
  bUsedWithMeshParticles=true
  bUsedWithNiagaraMeshParticles=true
  blendMode=BLEND_Translucent
  pixelDepthOffsetMode=PDOM_Legacy
  shadingModel=MSM_Unlit
  twoSided=true
  usedShadingModels=Unlit
 }
 nodes [
  n1 = ScalarParameter() {defaultValue=1, group=Texture/Detail, parameterName=Detail Tile B, sortPriority=6}
  n2 = Constant() {r=1}
  n3 = TextureCoordinate()
  n4 = MaterialFunctionCall(Float2=StaticSwitchParameter(True=MaterialFunctionCall(XY=@n3) {materialFunction=&Swizzle}.YX, False=@n3) {group=Pinch, parameterName=Horizontal/Vertical Switch}) {materialFunction=&BreakOutFloat2Components}
  n5 = LinearInterpolate(Alpha=@n4.R) {constA=-1, constB=2}
  n6 = StaticSwitchParameter(True=OneMinus(None=@n4.G), False=@n4.G) {group=Pinch, parameterName=Reverse}
  n7 = Saturate(None=AppendVector(A=@n4.G, B=LinearInterpolate(A=@n4.R, B=@n5, Alpha=@n6)))
  n8 = ScalarParameter() {defaultValue=0.5, group=Texture/Detail, parameterName=Detail PanSpeed B, sortPriority=7}
  n9 = AppendVector(A=StaticSwitchParameter(True=Multiply(A=@n8) {constB=-1}, False=@n8) {group=Pinch, parameterName=Reverse}, B=Constant())
  n10 = MaterialFunctionCall(Float2=Multiply(A=AppendVector(A=@n1, B=@n2), B=Panner(Coordinate=@n7, Speed=@n9) {speedX=1})) {materialFunction=&BreakOutFloat2Components}
  n11 = DynamicParameter() {defaultValue=<rgba 0 0 0 0>, paramNames [Random,0,0,0]}
  n12 = TextureSampleParameter2D(UVs=AppendVector(A=Add(A=@n10.R, B=@n11.Random), B=@n10.G)) {
   group=Texture/Detail
   parameterName=Detail Tex B
   samplerType=SAMPLERTYPE_LinearColor
   sortPriority=5
   texture=/Game/Effects/Fort_Effects/Textures/Tile/T_Raygun_Muzzle_02
  }
  n13 = StaticSwitchParameter(True=Constant2Vector() {g=0.8, r=0.5}, False=Constant2Vector() {g=0.2, r=0.5}) {group=Pinch, parameterName=Reverse}
  n14 = SphereMask(
   A=TextureCoordinate()
   B=StaticSwitchParameter(True=MaterialFunctionCall(XY=@n13) {materialFunction=&Swizzle}.YX, False=@n13) {group=Pinch, parameterName=Horizontal/Vertical Switch}
  ) {attenuationRadius=0.8, hardnessPercent=0.75}
  n15 = Multiply(A=@n12.G, B=@n14)
  n16 = ScalarParameter() {defaultValue=1, group=Texture/Detail, parameterName=Detail Emissive B, sortPriority=8}
  n17 = Saturate(None=@n16)
  n18 = ScalarParameter() {defaultValue=-1, group=Texture/Detail, parameterName=Detail PanSpeed A, sortPriority=2}
  n19 = AppendVector(A=StaticSwitchParameter(True=Multiply(A=@n18) {constB=-1}, False=@n18) {group=Pinch, parameterName=Reverse}, B=Constant())
  n20 = ScalarParameter() {defaultValue=1, group=Texture/Detail, parameterName=Detail Tile A, sortPriority=1}
  n21 = Constant() {r=1}
  n22 = MaterialFunctionCall(Float2=Multiply(A=Panner(Coordinate=@n7, Speed=@n19) {speedX=1}, B=AppendVector(A=@n20, B=@n21))) {materialFunction=&BreakOutFloat2Components}
  n23 = Add(A=@n22.R, B=@n11.Random)
  n24 = Multiply(
   A=TextureSampleParameter2D(UVs=AppendVector(A=@n23, B=@n22.G)) {group=Texture/Detail, parameterName=Detail Tex A, texture=/Game/Effects/Fort_Effects/Textures/Smoke/T_Smoke_ThinTrail_Packed}.G
   B=@n14
  )
  n25 = ScalarParameter() {defaultValue=3, group=Texture/Detail, parameterName=Detail Emissive A, sortPriority=3}
  n26 = Saturate(None=@n25)
  n27 = CurveAtlasRowParameter(CurveTime=Add(A=Multiply(A=@n15, B=@n17), B=Multiply(A=@n24, B=@n26))) {
   atlas=&AthenaCurveAtlas01
   curve=/Game/Athena/LinearColorCurves/Curve_Glider_Drift
   defaultValue=105
   group=Color
   parameterName=Color Curve
  }
  n28 = Multiply(A=@n16, B=Multiply(A=@n12.G, B=@n27))
  n29 = Multiply(A=Multiply(A=@n27, B=@n24), B=@n25)
  MaterialFunctionCall(Emissive=Add(A=@n27, B=Add(A=@n28, B=@n29))) {materialFunction=&MF_FX_EmissiveAdjustments}
  n30 = DynamicParameter() {defaultValue=<rgba 0 0 0 0>, paramNames [Random,0,0,0]}
  n31 = TextureSample(UVs=Panner(Coordinate=TextureCoordinate() {uTiling=0.4, vTiling=0.35}) {speedX=0.1, speedY=0.2}) {samplerType=SAMPLERTYPE_Normal, texture=&T_Weap_Distort_N_02}
  n32 = Panner(
   Coordinate=Multiply(A=ComponentMask(None=@n31.RGB) {g=true, r=true}) {constB=0.06}
   Speed=ScalarParameter() {defaultValue=-0.2, group=WPO, parameterName=Noise Evolution Speed}
  )
  n33 = Multiply(A=ComponentMask(None=TextureCoordinate()) {g=true}, B=ScalarParameter() {defaultValue=0.4, group=WPO, parameterName=X Frequency Mult})
  n34 = TextureSample(UVs=Add(A=@n30.Random, B=Add(A=@n32, B=@n33))) {samplerType=SAMPLERTYPE_LinearColor, texture=&T_NoisePack_LinearColor}
  n35 = Multiply(A=ComponentMask(None=TextureCoordinate()) {g=true}, B=ScalarParameter() {defaultValue=0.4, group=WPO, parameterName=Y Frequency Mult})
  n36 = TextureSample(UVs=Add(A=@n30.Random, B=Add(A=@n32, B=@n35))) {samplerType=SAMPLERTYPE_LinearColor, texture=&T_NoisePack_LinearColor}
  n37 = Multiply(A=ComponentMask(None=TextureCoordinate()) {g=true}, B=ScalarParameter() {defaultValue=0.4, group=WPO, parameterName=Z Frequency Mult})
  n38 = TextureSample(UVs=Add(A=@n30.Random, B=Add(A=@n32, B=@n37))) {samplerType=SAMPLERTYPE_LinearColor, texture=&T_NoisePack_LinearColor}
  n39 = MaterialFunctionCall(Float3=ConstantBiasScale(None=MaterialFunctionCall(R=@n34.R, G=@n36.G, B=@n38.B) {materialFunction=&AppendMany}.RGB) {bias=-0.5, scale=2}) {materialFunction=&BreakOutFloat3Components}
  n40 = Multiply(A=@n39.R, B=ScalarParameter() {defaultValue=20, group=WPO, parameterName=X_Intesity}) {constB=0.5}
  n41 = Multiply(A=@n39.G, B=ScalarParameter() {defaultValue=20, group=WPO, parameterName=Y_Intesity}) {constB=5}
  n42 = Multiply(A=@n39.B, B=ScalarParameter() {defaultValue=20, group=WPO, parameterName=Z_Intensity}) {constB=0.35}
  n43 = Constant2Vector()
  n44 = Multiply(A=ScalarParameter() {group=WPO, parameterName=Gravity}) {constB=-1}
  n45 = TextureCoordinate()
  n46 = StaticSwitchParameter(True=MaterialFunctionCall(XY=@n45) {materialFunction=&Swizzle}.YX, False=@n45) {group=Pinch, parameterName=Horizontal/Vertical Switch}
  n47 = StaticSwitchParameter(True=OneMinus(None=@n46), False=@n46) {group=Pinch, parameterName=Reverse}
  n48 = Add(
   A=Transform(None=MaterialFunctionCall(R=@n40, G=@n41, B=@n42) {materialFunction=&AppendMany}.RGB) {transformSourceType=TRANSFORMSOURCE_Local}
   B=Multiply(A=AppendVector(A=@n43, B=@n44), B=ComponentMask(None=@n47) {g=true})
  )
  n49 = Power(Base=ComponentMask(None=TextureCoordinate()) {g=true}, Exp=ScalarParameter() {defaultValue=1, group=WPO, parameterName=Fade Power})
  n50 = TextureCoordinate()
  n51 = TextureSample(UVs=Panner(Coordinate=MaterialFunctionCall(XY=@n50) {materialFunction=&Swizzle}.YX) {speedX=-0.65}) {texture=/Game/Characters/Fortress_CharacterShader/Textures/RainbowEffect}
  QualitySwitch(
   Default=ShadingPathSwitch(Default=Multiply(A=@n48, B=@n49)) {desc=Migrated from FeatureLevelSwitch}
   Low=Multiply(A=ConstantBiasScale(None=@n51.RGB) {bias=-0.5, scale=6}, B=@n49)
  )
  n52 = Constant()
  n53 = Multiply(A=@n24, B=If(A=@n25, B=@n52, "A > B"=Constant() {r=1}, "A == B"=@n52, "A < B"=@n52))
  n54 = TextureCoordinate()
  n55 = MaterialFunctionCall(Float2=StaticSwitchParameter(True=MaterialFunctionCall(XY=@n54) {materialFunction=&Swizzle}.YX, False=@n54) {group=Pinch, parameterName=Horizontal/Vertical Switch}) {materialFunction=&BreakOutFloat2Components}
  n56 = ScalarParameter() {
   defaultValue=-0.5
   desc="Pinches the \"head\" of the trail // Safe value from -.5 to .5"
   group=Pinch
   parameterName=Pinch Front
   sliderMax=0
   sliderMin=-0.5
  }
  n57 = ScalarParameter() {desc="Pinches the \"tail\" of the trail.", group=Pinch, parameterName=Pinch Rear, sliderMin=0.0001}
  n58 = StaticSwitchParameter(True=OneMinus(None=@n55.G), False=@n55.G) {group=Pinch, parameterName=Reverse}
  n59 = Saturate(None=LinearInterpolate(
    A=@n55.R
    B=LinearInterpolate(Alpha=@n55.R) {constA=-1, constB=2, desc=Keep the 0-1 range in the center}
    Alpha=LinearInterpolate(A=@n56, B=@n57, Alpha=@n58) {constB=1.6684}
   ))
  n60 = Constant() {r=-0.5}
  n61 = StaticSwitchParameter(True=@n60, False=Multiply(A=@n60) {constB=-1}) {group=Pinch, parameterName=Reverse}
  n62 = Constant()
  n63 = TextureSample(UVs=Panner(Coordinate=AppendVector(A=@n55.G, B=@n59), Speed=AppendVector(A=@n61, B=@n62)) {speedX=-1}) {texture=/Game/Effects/Fort_Effects/Textures/Tile/T_LaserTrail}
  n64 = Multiply(A=ParticleColor().A, B=Add(A=Saturate(None=@n53), B=Multiply(A=@n14, B=@n63.R) {constA=1}))
  n65 = Power(
   Base=MaterialFunctionCall(Distance=ScalarParameter() {defaultValue=25, group=CameraFade, parameterName=Near Camera Fade Dist}) {materialFunction=&Near_Camera_Fade}.Result
   Exp=ScalarParameter() {defaultValue=1, group=CameraFade, parameterName=Near Camera Fade Power}
  )
  QualitySwitch(Default=ShadingPathSwitch(Default=Multiply(A=@n64, B=@n65)) {desc=Migrated from FeatureLevelSwitch}, Low=@n64)
 ]
}
